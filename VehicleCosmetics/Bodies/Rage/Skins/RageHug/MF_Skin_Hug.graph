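# Material function for the Rage body's Hug skin, setting material attributes for an animated effect switchable between ice and spear variants.
# Time-driven sine, cosine and Fmod offsets slide packed RGBA masks on UV1, with a rotated, mirrored copy and panning noise, driving color, glow and ice roughness.
# These layer over diffuse, packed AO/roughness/metallic maps and a pulsing hue-shift Fresnel, with view-angle finish falloff.
# The AnimationOn static switch bypasses the animation for rendering.
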MaterialFunction /VehicleCosmetics/Bodies/Rage/Skins/RageHug/MF_Skin_Hug
graph {nodes [
  n1 = SetMaterialAttributes!(
   MaterialAttributes
   BaseColor=NamedRerouteUsage()
   Metallic=NamedRerouteUsage()
   Roughness=NamedRerouteUsage()
   EmissiveColor=NamedRerouteUsage()
   Normal=NamedRerouteUsage()
   AmbientOcclusion=NamedRerouteUsage()
  ) {attributeSetTypes [
    #1
    #2
    #3
    #4
    #5
    #6
   ]}
  FunctionOutput(None=@n1)
  n2 = TextureSampleParameter2D(UVs=StaticSwitchParameter(True=NamedRerouteUsage(), False=TextureCoordinate()) {defaultValue=true, group=BaseTextures, parameterName=DiffuseMap_UVSet2}) {
   group=BaseTextures
   parameterName=DiffuseAlphaMap
   samplerSource=SSM_Wrap_WorldGroupSettings
   texture=&/T_Hug_Diff
  }
  NamedRerouteDeclaration(None=@n2.RGB) {name=DiffuseOut, nodeColor=<rgba 1 0.1172 0 1>}
  NamedRerouteDeclaration(None=@n2.A) {name=DiffuseAlpha, nodeColor=<rgba 1 0.7734 0 1>}
  n3 = TextureSampleParameter2D(UVs=TextureCoordinate()) {
   desc="R=specular, G=ao, B=Roughness, A = Metallic"
   group=BaseTextures
   parameterName=AmbientOcclusion
   texture=/VehicleCosmetics/UniversalSkins/Endit/T_endit_AO_Fire
  }
  n4 = ScalarParameter() {defaultValue=0.5, group=Finish, parameterName=Roughness}
  n5 = StaticSwitchParameter(True=@n3.B, False=@n4) {desc=Roughness in B channel of AO map, group=BaseTextures, parameterName=Rough Map}
  n6 = ScalarParameter() {defaultValue=0.2, group=Spear, parameterName=speed}
  n7 = Add(A=NamedRerouteUsage(), B=ScalarParameter() {defaultValue=-0.2, group=Spear, parameterName=timing offset})
  n8 = Multiply(A=@n6, B=Add(A=@n7) {constB=0.35})
  n9 = Cosine(None=@n8)
  n10 = TextureCoordinate() {coordinateIndex=1}
  n11 = ComponentMask(None=@n10) {g=true}
  n12 = TextureSampleParameter2D(UVs=@n10) {parameterName=Noise, texture=/VehicleCosmetics/Bodies/Rage/Skins/RageHug2/T_Hug2_Noise}
  n13 = Multiply(A=Add(A=@n11, B=@n12.G), B=Saturate(None=Multiply(A=@n9) {constB=1.5}))
  n14 = Multiply(A=Saturate(None=ConstantBiasScale(None=@n13) {bias=-0.05, scale=2}), B=ConstantBiasScale(None=@n12.G) {bias=0.1, scale=1})
  n15 = Add(A=Clamp(None=Multiply(A=@n9) {constB=0.5}) {maxDefault=0.2}, B=Saturate(None=Multiply(A=@n14) {constB=2}))
  n16 = StaticSwitchParameter(True=LinearInterpolate(A=@n5, Alpha=@n15) {constB=0}, False=@n5) {defaultValue=true, desc="t=ice, f=spear", group=BaseTextures, parameterName=Ice or Spear}
  NamedRerouteDeclaration(None=Saturate(None=StaticSwitchParameter(True=@n16, False=@n4) {defaultValue=true, desc=Switch for Rendering, group=Rendering, parameterName=AnimationOn})) {name=Roughness, nodeColor=<rgba 0 0.1094 1 1>}
  n17 = Constant3Vector() {constant=<rgba 1 1 1 0>}
  n18 = ScalarParameter() {defaultValue=1, group=BaseTextures, parameterName=AO}
  NamedRerouteDeclaration(None=StaticSwitchParameter(True=LinearInterpolate(A=@n17, B=@n3.G, Alpha=@n18), False=@n17) {desc=AO in G channel of AO map, group=BaseTextures, parameterName=AmbientOcclusionMap}) {name=AO, nodeColor=<rgba 0.7969 0.7969 0.7969 1>}
  NamedRerouteDeclaration(None=StaticSwitchParameter(True=@n3.A, False=ScalarParameter() {group=Finish, parameterName=MetallicDefault}) {desc=Metallic in A channel of AO map, group=BaseTextures, parameterName=Metallic Map}) {name=Metallic, nodeColor=<rgba 0.4297 1 0 1>}
  n19 = NamedRerouteUsage()
  n20 = StaticSwitchParameter(
   True=VectorParameter() {group=Fresnel, parameterName=Color_Fresnel, defaultValue=<rgba 1 0 0.0976 1>}.RGB
   False=StaticSwitchParameter(True=NamedRerouteUsage(), False=NamedRerouteUsage()) {group=Fresnel, parameterName=UseCustomColor_Fresnel}
  ) {group=Fresnel, parameterName=UseUniqueColor_Fresnel}
  n21 = TextureCoordinate() {coordinateIndex=1}
  n22 = ScalarParameter() {defaultValue=1, group=Fresnel, parameterName=MultichromeScale_fresnel}
  n23 = ScalarParameter() {defaultValue=0.2, group=Fresnel, parameterName=MultichromePanSpeed_Fresnel}
  n24 = NamedRerouteUsage()
  n25 = ComponentMask(None=Panner(Coordinate=Multiply(A=@n21, B=@n22), Time=Multiply(A=@n23, B=@n24)) {speedY=1}) {g=true}
  n26 = StaticSwitchParameter(
   True=LinearInterpolate(
    A=@n20
    B=MaterialFunctionCall(Hue Shift Percentage=@n25, Texture=@n20) {materialFunction=&HueShift}.Result
    Alpha=ScalarParameter() {defaultValue=1, group=Fresnel, parameterName=MultichromeColor_Fresnel}
   )
   False=@n20
  ) {defaultValue=true, group=Fresnel, parameterName=UseMultichromeColor_Fresnel}
  n27 = DotProduct(A=Transform(None=CameraVectorWS()) {transformSourceType=TRANSFORMSOURCE_World, transformType=TRANSFORM_Tangent}, B=NamedRerouteUsage())
  n28 = Multiply(
   A=Power(Base=OneMinus(None=@n27), Exp=ScalarParameter() {defaultValue=5, group=Fresnel, parameterName=Fresnel_Power}) {constExponent=3}
   B=NamedRerouteUsage()
  )
  n29 = Sine(None=Multiply(A=ScalarParameter() {defaultValue=1, group=Fresnel, parameterName=EmissivePulseTiming_Fresnel}, B=NamedRerouteUsage())) {period=3}
  n30 = ScalarParameter() {defaultValue=0.2, group=Fresnel, parameterName=EmissiveIntensity_Fresnel}
  n31 = Multiply(
   A=LinearInterpolate(B=ConstantBiasScale(None=@n29) {bias=1.25}, Alpha=ScalarParameter() {defaultValue=1, group=Fresnel, parameterName=PulseInfluence_Fresnel}) {constA=1}
   B=@n30
  )
  n32 = LinearInterpolate(
   A=Multiply(A=@n19, B=ScalarParameter() {defaultValue=0.1, group=Colors, parameterName=DiffuseEmissiveIntensity}) {desc=Emissive}
   B=@n26
   Alpha=Multiply(A=@n28, B=StaticSwitchParameter(True=@n31, False=@n30) {defaultValue=true, group=Fresnel, parameterName=EmissivePulse_Fresnel})
  ) {constAlpha=1, desc=Emissive}
  n33 = VectorParameter() {group=Spear, parameterName=Color, defaultValue=<rgba 0.2983 0.7064 1 1>}
  n34 = Multiply(A=@n33.RGB) {constB=0.15}
  n35 = Abs(None=Sine(None=NamedRerouteUsage()) {period=0.2})
  n36 = Add(B=@n11) {constA=-0.1}
  n37 = ScalarParameter() {defaultValue=0.2, group=Spear, parameterName=spear offset}
  n38 = Multiply(A=NamedRerouteUsage(), B=ScalarParameter() {defaultValue=0.2, group=Spear, parameterName=speed})
  n39 = Fmod(A=Add(A=Multiply(A=@n38) {constB=3}) {constB=10.6}, B=Constant() {r=3})
  n40 = Clamp(None=ConstantBiasScale(None=OneMinus(None=@n39)) {bias=-0.5, scale=3}) {minDefault=-1}
  n41 = ScalarParameter() {defaultValue=1.1, group=Spear, parameterName=SpearRotate}
  n42 = OneMinus(None=@n10)
  n43 = Multiply(A=@n41) {constB=-1}
  n44 = ComponentMask(None=@n10) {r=true}
  n45 = LinearInterpolate(
   A=ComponentMask(None=Rotator(Coordinate=@n10, Time=@n41)) {r=true}
   B=ComponentMask(None=Rotator(Coordinate=@n42, Time=@n43)) {r=true}
   Alpha=Saturate(None=ConstantBiasScale(None=@n44) {bias=-0.5, scale=10000})
  )
  n46 = TextureSampleParameter2D(UVs=Add(A=AppendVector(A=@n37, B=@n40), B=AppendVector(A=@n45, B=@n11))) {parameterName=RGBA_Mask, texture=/VehicleCosmetics/Bodies/Rage/Skins/RageHug2/T_Hug2_RGB}
  n47 = Constant()
  n48 = Multiply(A=@n7, B=@n6)
  n49 = Fmod(A=Add(A=Multiply(A=@n48) {constB=3}) {constB=10.6}, B=Constant() {r=3})
  n50 = Clamp(None=ConstantBiasScale(None=OneMinus(None=@n49)) {bias=-0.5, scale=4}) {minDefault=-1}
  n51 = TextureSampleParameter2D(UVs=Add(A=TextureCoordinate() {coordinateIndex=1}, B=AppendVector(A=@n47, B=@n50))) {parameterName=RGBA_Mask, texture=&T_Hug2_RGB}
  n52 = Max(A=@n46.R, B=@n51.R) {desc=Icicles - RedChannels}
  n53 = ConstantBiasScale(None=Multiply(A=Add(A=@n35, B=@n36), B=@n52)) {bias=-0.2, scale=5}
  n54 = LinearInterpolate(A=@n32, B=@n34, Alpha=Saturate(None=@n53)) {desc=Diffuse}
  n55 = NamedRerouteUsage()
  n56 = LinearInterpolate(A=Multiply(A=@n55) {constB=0.15}, B=@n55, Alpha=Saturate(None=Multiply(A=@n14) {constB=1.5}))
  n57 = Transform(None=CameraVectorWS()) {transformSourceType=TRANSFORMSOURCE_World, transformType=TRANSFORM_Tangent}
  n58 = NamedRerouteUsage()
  n59 = Power(Base=OneMinus(None=DotProduct(A=@n57, B=@n58))) {constExponent=6}
  n60 = LinearInterpolate(A=@n15, B=Clamp(None=@n59) {minDefault=0.005}) {constAlpha=0.95}
  n61 = ScalarParameter() {group=Fresnel, parameterName=DiffuseIntensity_Fresnel}
  n62 = LinearInterpolate(A=@n19, B=@n26, Alpha=Saturate(None=Multiply(A=@n28, B=@n61))) {constAlpha=1, desc=Diffuse}
  n63 = NamedRerouteUsage()
  n64 = Multiply(A=@n63, B=ScalarParameter() {defaultValue=0.2, group=Spear, parameterName=speed})
  n65 = ConstantBiasScale(None=Sine(None=@n64)) {bias=0.65, scale=4}
  n66 = Add(A=TextureCoordinate() {coordinateIndex=1}, B=AppendVector(A=Constant(), B=Clamp(None=@n65) {maxDefault=2, minDefault=0.105}))
  n67 = Cosine(None=@n64)
  n68 = OneMinus(None=Saturate(None=ConstantBiasScale(None=@n67) {bias=-0.5, scale=10}))
  n69 = TextureSampleParameter2D(UVs=@n66) {parameterName=RGBA_Mask, texture=&T_Hug2_RGB}
  n70 = TextureSampleParameter2D(UVs=Panner(Coordinate=TextureCoordinate() {coordinateIndex=1, uTiling=0.1, vTiling=0.15}, Time=@n63) {speedX=0.1}) {parameterName=RGBA_Mask, texture=&T_Hug2_RGB}
  n71 = Multiply(A=Multiply(A=Multiply(A=@n68, B=@n69.G), B=ConstantBiasScale(None=@n70.B) {bias=-0.2, scale=1.25})) {constB=0.12}
  n72 = Constant()
  n73 = TextureSampleParameter2D(UVs=Add(A=@n66, B=AppendVector(A=@n71, B=@n72))) {parameterName=RGBA_Mask, texture=&T_Hug2_RGB}
  n74 = TextureCoordinate() {coordinateIndex=1}
  n75 = ScalarParameter() {defaultValue=1.1, group=Spear, parameterName=SpearRotate}
  n76 = ComponentMask(None=Rotator(Coordinate=@n74, Time=@n75)) {r=true}
  n77 = ComponentMask(None=Rotator(Coordinate=OneMinus(None=@n74), Time=Multiply(A=@n75) {constB=-1})) {r=true}
  n78 = Saturate(None=ConstantBiasScale(None=ComponentMask(None=@n74) {r=true}) {bias=-0.5, scale=10000})
  n79 = NamedRerouteUsage()
  n80 = Multiply(
   A=Add(A=@n79, B=ScalarParameter() {defaultValue=-0.2, group=Spear, parameterName=timing offset})
   B=ScalarParameter() {defaultValue=0.2, group=Spear, parameterName=speed}
  )
  n81 = ConstantBiasScale(None=Sine(None=@n80)) {bias=0.7, scale=7}
  n82 = Add(
   A=AppendVector(A=LinearInterpolate(A=@n76, B=@n77, Alpha=@n78), B=ComponentMask(None=@n74) {g=true})
   B=AppendVector(A=ScalarParameter() {defaultValue=0.2, group=Spear, parameterName=spear offset}, B=Clamp(None=@n81) {maxDefault=2, minDefault=0.15})
  )
  n83 = Cosine(None=@n80)
  n84 = OneMinus(None=Saturate(None=ConstantBiasScale(None=@n83) {bias=-0.5, scale=10}))
  n85 = TextureSampleParameter2D(UVs=@n82) {parameterName=RGBA_Mask, texture=&T_Hug2_RGB}
  n86 = TextureSampleParameter2D(UVs=Panner(Coordinate=TextureCoordinate() {coordinateIndex=1, uTiling=0.1, vTiling=0.15}, Time=@n79) {speedX=-0.05}) {parameterName=RGBA_Mask, texture=&T_Hug2_RGB}
  n87 = Multiply(A=Multiply(A=Multiply(A=@n84, B=@n85.G), B=ConstantBiasScale(None=@n86.B) {bias=-0.2, scale=1.25})) {constB=0.12}
  n88 = Constant()
  n89 = TextureSampleParameter2D(UVs=Add(A=@n82, B=AppendVector(A=@n87, B=@n88))) {parameterName=RGBA_Mask, texture=&T_Hug2_RGB}
  n90 = Max(A=@n73.R, B=@n89.R) {desc=Spears - Red Channels}
  n91 = Saturate(None=Multiply(A=@n90) {constB=30})
  n92 = LinearInterpolate(A=@n62, B=Constant3Vector() {constant=<rgba 0.03 0.03 0.03 0>}, Alpha=@n91) {constAlpha=1, desc=Diffuse}
  n93 = VectorParameter() {group=Spear, parameterName=Color, defaultValue=<rgba 0.2983 0.7064 1 1>}
  n94 = StaticSwitchParameter(
   True=StaticSwitchParameter(
    True=LinearInterpolate(A=@n54, B=@n56, Alpha=@n60) {desc=Diffuse}
    False=LinearInterpolate(A=@n92, B=@n93.RGB, Alpha=@n90) {constAlpha=1, desc=Diffuse}
   ) {defaultValue=true, desc="t=ice, f=spear", group=BaseTextures, parameterName=Ice or Spear}
   False=@n62
  ) {defaultValue=true, desc=Switch for Rendering, group=Rendering, parameterName=AnimationOn}
  n95 = NamedRerouteUsage()
  n96 = Transform(None=CameraVectorWS()) {transformSourceType=TRANSFORMSOURCE_World, transformType=TRANSFORM_Tangent}
  n97 = NamedRerouteUsage()
  n98 = Power(Base=OneMinus(None=DotProduct(A=@n96, B=@n97))) {constExponent=3}
  n99 = OneMinus(None=@n98)
  n100 = ScalarParameter() {defaultValue=1, group=Finish, parameterName=FinishInteriorFalloff}
  n101 = ScalarParameter() {defaultValue=1, group=Finish, parameterName=FinishOuterFalloff}
  n102 = Max(A=Saturate(None=Power(Base=@n99, Exp=@n100)), B=Saturate(None=Power(Base=@n98, Exp=@n101)))
  n103 = ScalarParameter() {defaultValue=95, group=Finish, parameterName=MetallicFinishInteriorFalloff}
  n104 = Clamp(None=Power(Base=OneMinus(None=@n98), Exp=@n103), Min=ScalarParameter() {defaultValue=0.1, group=Finish, parameterName=MetallicDarkMinLevel})
  n105 = ScalarParameter() {defaultValue=5, group=Finish, parameterName=MetallicFinishOuterFalloff}
  n106 = Max(A=@n104, B=Saturate(None=Power(Base=@n98, Exp=@n105)))
  n107 = Power(Base=@n104, Exp=@n103)
  n108 = ScalarParameter() {defaultValue=0.1, group=Finish, parameterName=MetallicHighlight}
  n109 = LinearInterpolate(A=Multiply(A=@n94, B=Multiply(A=@n95, B=@n102)), B=Add(A=Multiply(A=@n94, B=@n106), B=Multiply(A=@n107, B=@n108)), Alpha=NamedRerouteUsage())
  NamedRerouteDeclaration(None=Saturate(None=ConstantBiasScale(None=@n109) {bias=-0.01, desc=Color correction for BR, scale=1.01})) {name=BaseColor, nodeColor=<rgba 1 0 0.375 1>}
  n110 = Saturate(None=Max(A=ConstantBiasScale(None=@n46.A) {bias=-0.01, scale=1}, B=ConstantBiasScale(None=@n51.A) {bias=-0.01, scale=1}))
  n111 = TextureCoordinate() {coordinateIndex=1, uTiling=0.5, vTiling=0.3}
  n112 = Constant2Vector() {g=2, r=3}
  n113 = NamedRerouteUsage()
  n114 = TextureSampleParameter2D(UVs=Panner(Coordinate=Multiply(A=@n111, B=@n112), Time=@n113) {speedX=-0.02, speedY=0.1}) {parameterName=Noise, texture=&T_Hug2_Noise}
  n115 = Multiply(A=@n110, B=@n114.B)
  n116 = Multiply(A=@n115) {constB=100}
  n117 = ConstantBiasScale(None=TextureSampleParameter2D(UVs=Panner(Coordinate=@n111, Time=@n113) {speedX=-0.02, speedY=0.1}) {parameterName=Noise, texture=&T_Hug2_Noise}.A) {bias=-0.5, scale=2}
  n118 = TextureSampleParameter2D(UVs=Panner(Coordinate=TextureCoordinate() {coordinateIndex=1}, Time=@n113) {speedX=-0.02, speedY=0.1}) {parameterName=Noise, texture=&T_Hug2_Noise}
  n119 = Multiply(A=@n118.R) {constB=0.5}
  n120 = Max(A=Max(A=@n46.B, B=@n51.B) {desc=Icicles - BlueChannels}, B=Add(A=Multiply(A=@n110, B=@n117), B=Subtract(A=@n110, B=@n119)))
  n121 = Saturate(None=Max(A=Max(A=@n53, B=@n116), B=Multiply(A=@n120) {constB=2}))
  n122 = Multiply(A=@n115) {constB=20}
  n123 = Multiply(A=Multiply(A=Clamp(None=@n59) {minDefault=0.005}, B=@n15)) {constB=5}
  n124 = Multiply(A=ConstantBiasScale(None=@n59) {bias=-0.5, scale=2}, B=Saturate(None=@n9))
  n125 = LinearInterpolate(
   A=LinearInterpolate(A=LinearInterpolate(A=@n32, B=@n34, Alpha=@n121) {desc=Emissive}, B=@n33.RGB, Alpha=Clamp(None=@n122) {maxDefault=2}) {desc=Emissive}
   B=@n55
   Alpha=Saturate(None=Max(A=@n123, B=@n124))
  ) {desc=Emissive}
  n126 = NamedRerouteUsage()
  n127 = Sine(None=@n8) {period=0.5}
  n128 = TextureSampleParameter2D(UVs=TextureCoordinate() {coordinateIndex=1}) {parameterName=RGBA_Mask, texture=&T_Hug2_RGB}
  n129 = Multiply(A=ConstantBiasScale(None=OneMinus(None=@n127)) {bias=1.5}, B=@n128.G)
  n130 = ConstantBiasScale(None=Multiply(A=OneMinus(None=@n59), B=Multiply(A=@n126, B=@n129))) {bias=-0.02, scale=5}
  n131 = Multiply(A=@n128.G) {constB=2}
  n132 = Multiply(A=ConstantBiasScale(None=@n127) {bias=1.5}, B=Add(A=@n128.G, B=Multiply(A=@n114.B, B=@n131)))
  n133 = ConstantBiasScale(None=Subtract(A=@n128.G, B=Multiply(A=@n118.A) {constB=0.2})) {bias=-0.05, scale=2}
  n134 = LinearInterpolate(
   A=LinearInterpolate(A=LinearInterpolate(A=@n125, B=@n33.RGB, Alpha=@n52) {desc=Emissive}, B=@n33.RGB, Alpha=Clamp(None=@n130) {maxDefault=2}) {desc=Emissive}
   B=@n55
   Alpha=Saturate(None=Max(A=@n132, B=@n133))
  ) {desc=Emissive}
  n135 = NamedRerouteUsage()
  n136 = Floor(None=@n66)
  n137 = Saturate(None=ConstantBiasScale(None=ComponentMask(None=@n136) {g=true}))
  n138 = ConstantBiasScale(None=@n67) {bias=-0.25, scale=2}
  n139 = Multiply(A=Multiply(A=OneMinus(None=@n137), B=@n69.A), B=OneMinus(None=Saturate(None=@n138)))
  n140 = Floor(None=@n82)
  n141 = Saturate(None=ConstantBiasScale(None=ComponentMask(None=@n140) {g=true}))
  n142 = Multiply(A=@n83) {constB=2}
  n143 = Multiply(A=Multiply(A=OneMinus(None=@n141), B=@n85.A), B=OneMinus(None=Saturate(None=@n142)))
  n144 = Max(A=@n139, B=@n143) {desc=Spears - AlphaFade1}
  n145 = LinearInterpolate(
   A=LinearInterpolate(A=@n32, B=Multiply(A=@n93.RGB) {constB=0.25}, Alpha=@n91) {desc=emissive}
   B=@n135
   Alpha=Multiply(A=@n144) {constB=0.25}
  ) {desc=emissive}
  n146 = NamedRerouteUsage()
  n147 = TextureCoordinate() {coordinateIndex=1, vTiling=0.02}
  n148 = TextureCoordinate() {coordinateIndex=1, vTiling=0.05}
  n149 = Subtract(
   A=TextureSampleParameter2D(UVs=Panner(Coordinate=@n147) {speedX=0.2}) {parameterName=RGBA_Mask, texture=&T_Hug2_RGB}.B
   B=TextureSampleParameter2D(UVs=Panner(Coordinate=@n148) {speedX=-0.2}) {parameterName=RGBA_Mask, texture=&T_Hug2_RGB}.B
  )
  n150 = Saturate(None=Abs(None=ConstantBiasScale(None=@n149) {bias=-0.1}))
  n151 = Saturate(None=ConstantBiasScale(None=Subtract(A=@n144, B=@n150)) {bias=-0.12, scale=0.25})
  n152 = TextureSampleParameter2D(UVs=Panner(Coordinate=TextureCoordinate() {coordinateIndex=1, uTiling=6, vTiling=1.5}, Time=@n79) {speedY=1}) {parameterName=Noise, texture=&T_Hug2_Noise}
  n153 = Add(A=@n152.A) {constB=0.5}
  n154 = TextureSampleParameter2D(UVs=Panner(Coordinate=TextureCoordinate() {coordinateIndex=1, uTiling=2, vTiling=0.6}, Time=@n79) {speedX=0.2, speedY=1}) {parameterName=Noise, texture=&T_Hug2_Noise}
  n155 = TextureCoordinate() {coordinateIndex=1, uTiling=2, vTiling=0.2}
  n156 = OneMinus(None=TextureSampleParameter2D(UVs=Panner(Coordinate=@n155, Time=@n79) {speedY=0.4}) {parameterName=Noise, texture=&T_Hug2_Noise}.A)
  n157 = ConstantBiasScale(None=Multiply(A=@n144, B=Multiply(A=@n154.G, B=@n156))) {bias=-0.02, scale=15}
  n158 = LinearInterpolate(
   A=LinearInterpolate(
    A=LinearInterpolate(A=@n145, B=@n146, Alpha=@n151) {desc=emissive}
    B=LinearInterpolate(A=@n135, B=@n146, Alpha=@n153)
    Alpha=Saturate(None=@n157)
   ) {desc=emissive}
   B=Constant3Vector()
   Alpha=Multiply(A=@n91) {constB=0.9}
  ) {desc=emissive}
  n159 = Clamp(None=Multiply(A=@n90) {constB=1.5}) {maxDefault=1000}
  n160 = OneMinus(None=Saturate(None=ConstantBiasScale(None=@n67) {bias=-0.25, scale=2}))
  n161 = OneMinus(None=Saturate(None=ConstantBiasScale(None=@n83) {bias=0, scale=2}))
  n162 = ConstantBiasScale(None=Max(A=Multiply(A=@n73.R, B=@n160), B=Multiply(A=@n89.R, B=@n161))) {bias=-0.3, scale=2}
  n163 = Multiply(A=@n156, B=@n152.B)
  n164 = Multiply(A=@n152.R, B=@n154.B)
  n165 = ConstantBiasScale(None=Multiply(A=@n144, B=Add(A=@n163, B=@n164))) {bias=-0.1, scale=10}
  n166 = Multiply(A=@n85.G) {constB=60}
  n167 = Max(
   A=ConstantBiasScale(None=Subtract(A=@n139, B=@n69.G)) {bias=-0.3, scale=2}
   B=ConstantBiasScale(None=Subtract(A=@n143, B=@n166) {constB=0}) {bias=-0.3, scale=1}
  ) {constB=0}
  n168 = LinearInterpolate(
   A=LinearInterpolate(A=LinearInterpolate(A=@n158, B=@n93.RGB, Alpha=@n159) {desc=emissive}, B=NamedRerouteUsage(), Alpha=Saturate(None=@n162)) {desc=emissive}
   B=NamedRerouteUsage()
   Alpha=Clamp(None=Max(A=@n165, B=@n167) {constB=0}) {maxDefault=5}
  ) {desc=emissive}
  n169 = Saturate(None=StaticSwitchParameter(
    True=StaticSwitchParameter(True=@n134, False=@n168) {defaultValue=true, desc="t=ice, f=spear", group=BaseTextures, parameterName=Ice or Spear}
    False=@n32
   ) {defaultValue=true, desc=Switch for Rendering, group=Rendering, parameterName=AnimationOn})
  n170 = ScalarParameter() {defaultValue=3, group=Colors, parameterName=EmissiveIntensity}
  NamedRerouteDeclaration(None=ConstantBiasScale(None=Multiply(A=@n169, B=@n170)) {bias=-0.015, desc=Color correction for BR, scale=1.15}) {name=EmissiveColor, nodeColor=<rgba 0 0.6953 1 1>}
  NamedRerouteDeclaration(None=StaticSwitchParameter(True=@n3.R, False=ScalarParameter() {defaultValue=1, group=Finish, parameterName=SpecularHighlightIntensity}) {desc="Spec in R channel of AO map (helps Finish consistency)", group=BaseTextures, parameterName=SpecMap}) {name=Specular, nodeColor=<rgba 1 0.2578 0 1>}
  n171 = TextureSampleParameter2D(UVs=TextureCoordinate()) {
   group=BaseTextures
   parameterName=NormalMap
   samplerSource=SSM_Wrap_WorldGroupSettings
   samplerType=SAMPLERTYPE_Normal
   texture=&/T_Hug_N
  }
  NamedRerouteDeclaration(None=StaticSwitchParameter(True=@n171.RGB, False=@n171.RGB) {defaultValue=true, desc="t=ice, f=spear", group=BaseTextures, parameterName=Ice or Spear}) {name=Normals, nodeColor=<rgba 0.25 0.25 1 1>}
  n172 = TextureCoordinate()
  n173 = VectorParameter() {desc="(R,G) = Detail1    (B,A) = Detail2", group=BaseTextures, parameterName=Universal_Map_Transform, defaultValue=<rgba 1 1 0 0>}
  n174 = ComponentMask(None=@n173.RGB) {g=true, r=true}
  NamedRerouteDeclaration(None=Add(A=Multiply(A=@n172, B=@n174), B=AppendVector(A=@n173.B, B=@n173.B)) {desc=Panner}) {name=DiffuseNormal UVs, nodeColor=<rgba 0.1484 0 1 1>}
  n175 = VectorParameter() {desc="Alpha = Emissive Color Brightness", group=Colors, parameterName=ColorA, defaultValue=<rgba 1 0.2673 0 1>}
  NamedRerouteDeclaration(None=@n175.RGB) {name=TeamColor, nodeColor=<rgba 0 0.5312 1 1>}
  NamedRerouteDeclaration(None=VectorParameter() {group=Colors, parameterName=ColorB, defaultValue=<rgba 1 0 0 1>}.RGB) {name=CustomColor, nodeColor=<rgba 0 1 0.0625 1>}
  NamedRerouteDeclaration(None=Time() {bOverride_Period=true, period=500}) {name=time, nodeColor=<rgba 1 0 0.4219 1>}
 ]}
