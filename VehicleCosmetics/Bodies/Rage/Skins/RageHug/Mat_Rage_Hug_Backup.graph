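# Backup vehicle body skin material for Rage Hug, with animated ice or spear effects chosen by the "Ice or Spear" static switch, defaulting to ice.
# Time-driven UV offsets slide RGBA mask layers on UV1, one rotated and mirrored, while panning noise adds frost or spear trails.
# These masks blend over diffuse, pulsing hue-shifted Fresnel and falloff-based finish to drive base color and emissive.
# An optional packed map supplies specular, AO, roughness and metallic.
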Material /VehicleCosmetics/Bodies/Rage/Skins/RageHug/Mat_Rage_Hug_Backup
graph {
 properties {bAutomaticallySetUsageInEditor=false, usedShadingModels=Default Lit}
 nodes [
  n1 = TextureSampleParameter2D(UVs=StaticSwitchParameter(True=NamedRerouteUsage(), False=TextureCoordinate()) {defaultValue=true, group=BaseTextures, parameterName=DiffuseMap_UVSet2}) {
   group=BaseTextures
   parameterName=DiffuseAlphaMap
   samplerSource=SSM_Wrap_WorldGroupSettings
   texture=&/T_Hug_Diff
  }
  n2 = StaticSwitchParameter(
   True=VectorParameter() {group=Fresnel, parameterName=Color_Fresnel, defaultValue=<rgba 1 0 0.0976 1>}.RGB
   False=StaticSwitchParameter(True=NamedRerouteUsage(), False=NamedRerouteUsage()) {group=Fresnel, parameterName=UseCustomColor_Fresnel}
  ) {group=Fresnel, parameterName=UseUniqueColor_Fresnel}
  n3 = TextureCoordinate() {coordinateIndex=1}
  n4 = ScalarParameter() {defaultValue=1, group=Fresnel, parameterName=MultichromeScale_fresnel}
  n5 = ScalarParameter() {defaultValue=0.2, group=Fresnel, parameterName=MultichromePanSpeed_Fresnel}
  n6 = NamedRerouteUsage()
  n7 = ComponentMask(None=Panner(Coordinate=Multiply(A=@n3, B=@n4), Time=Multiply(A=@n5, B=@n6)) {speedY=1}) {g=true}
  n8 = StaticSwitchParameter(
   True=LinearInterpolate(
    A=@n2
    B=MaterialFunctionCall(Hue Shift Percentage=@n7, Texture=@n2) {materialFunction=&HueShift}.Result
    Alpha=ScalarParameter() {defaultValue=1, group=Fresnel, parameterName=MultichromeColor_Fresnel}
   )
   False=@n2
  ) {defaultValue=true, group=Fresnel, parameterName=UseMultichromeColor_Fresnel}
  n9 = DotProduct(A=Transform(None=CameraVectorWS()) {transformSourceType=TRANSFORMSOURCE_World, transformType=TRANSFORM_Tangent}, B=NamedRerouteUsage())
  n10 = Multiply(
   A=Power(Base=OneMinus(None=@n9), Exp=ScalarParameter() {defaultValue=5, group=Fresnel, parameterName=Fresnel_Power}) {constExponent=3}
   B=NamedRerouteUsage()
  )
  n11 = Sine(None=Multiply(A=ScalarParameter() {defaultValue=1, group=Fresnel, parameterName=EmissivePulseTiming_Fresnel}, B=NamedRerouteUsage())) {period=3}
  n12 = ScalarParameter() {defaultValue=0.2, group=Fresnel, parameterName=EmissiveIntensity_Fresnel}
  n13 = Multiply(
   A=LinearInterpolate(B=ConstantBiasScale(None=@n11) {bias=1.25}, Alpha=ScalarParameter() {defaultValue=1, group=Fresnel, parameterName=PulseInfluence_Fresnel}) {constA=1}
   B=@n12
  )
  n14 = LinearInterpolate(
   A=Multiply(A=@n1.RGB, B=ScalarParameter() {defaultValue=0.1, group=Colors, parameterName=DiffuseEmissiveIntensity}) {desc=Emissive}
   B=@n8
   Alpha=Multiply(A=@n10, B=StaticSwitchParameter(True=@n13, False=@n12) {defaultValue=true, group=Fresnel, parameterName=EmissivePulse_Fresnel})
  ) {constAlpha=1, desc=Emissive}
  n15 = VectorParameter() {group=Spear, parameterName=Color, defaultValue=<rgba 0.2983 0.7064 1 1>}
  n16 = ConstantBiasScale(None=@n15.RGB) {bias=0, scale=0.15}
  n17 = Abs(None=Sine(None=NamedRerouteUsage()) {period=0.2})
  n18 = TextureCoordinate() {coordinateIndex=1}
  n19 = ComponentMask(None=@n18) {g=true}
  n20 = ConstantBiasScale(None=@n19) {bias=-0.1, scale=1}
  n21 = ScalarParameter() {defaultValue=0.2, group=Spear, parameterName=spear offset}
  n22 = Multiply(A=NamedRerouteUsage(), B=ScalarParameter() {defaultValue=0.2, group=Spear, parameterName=speed})
  n23 = Fmod(A=Add(A=Multiply(A=@n22) {constB=3}) {constB=10.6}, B=Constant() {r=3})
  n24 = Clamp(None=ConstantBiasScale(None=OneMinus(None=@n23)) {bias=-0.5, scale=3}) {minDefault=-1}
  n25 = ScalarParameter() {defaultValue=1.1, group=Spear, parameterName=SpearRotate}
  n26 = OneMinus(None=@n18)
  n27 = Multiply(A=@n25) {constB=-1}
  n28 = ComponentMask(None=@n18) {r=true}
  n29 = LinearInterpolate(
   A=ComponentMask(None=Rotator(Coordinate=@n18, Time=@n25)) {r=true}
   B=ComponentMask(None=Rotator(Coordinate=@n26, Time=@n27)) {r=true}
   Alpha=Clamp(None=ConstantBiasScale(None=@n28) {bias=-0.5, scale=10000})
  )
  n30 = TextureSampleParameter2D(UVs=Add(A=AppendVector(A=@n21, B=@n24), B=AppendVector(A=@n29, B=@n19))) {parameterName=RGBA_Mask, texture=/VehicleCosmetics/Bodies/Rage/Skins/RageHug2/T_Hug2_RGB}
  n31 = Constant()
  n32 = Add(A=NamedRerouteUsage(), B=ScalarParameter() {defaultValue=-0.2, group=Spear, parameterName=timing offset})
  n33 = ScalarParameter() {defaultValue=0.2, group=Spear, parameterName=speed}
  n34 = Multiply(A=@n32, B=@n33)
  n35 = Fmod(A=Add(A=Multiply(A=@n34) {constB=3}) {constB=10.6}, B=Constant() {r=3})
  n36 = Clamp(None=ConstantBiasScale(None=OneMinus(None=@n35)) {bias=-0.5, scale=4}) {minDefault=-1}
  n37 = TextureSampleParameter2D(UVs=Add(A=TextureCoordinate() {coordinateIndex=1}, B=AppendVector(A=@n31, B=@n36))) {parameterName=RGBA_Mask, texture=&T_Hug2_RGB}
  n38 = Max(A=@n30.R, B=@n37.R) {desc=Icicles - RedChannels}
  n39 = ConstantBiasScale(None=Multiply(A=Add(A=@n17, B=@n20), B=@n38)) {bias=-0.2, scale=5}
  n40 = Clamp(None=@n39)
  n41 = TextureSampleParameter2D(UVs=@n18) {parameterName=Noise, texture=/VehicleCosmetics/Bodies/Rage/Skins/RageHug2/T_Hug2_Noise}
  n42 = Multiply(A=@n33, B=Add(A=@n32) {constB=0.35})
  n43 = Cosine(None=@n42)
  n44 = Multiply(A=Add(A=@n19, B=@n41.G), B=Clamp(None=ConstantBiasScale(None=@n43) {bias=0, scale=1.5}))
  n45 = Multiply(A=Clamp(None=ConstantBiasScale(None=@n44) {bias=-0.05, scale=2}), B=ConstantBiasScale(None=@n41.G) {bias=0.1, scale=1})
  n46 = Clamp(None=ConstantBiasScale(None=@n45) {bias=0, scale=1.5})
  n47 = Add(A=Clamp(None=ConstantBiasScale(None=@n43) {bias=0}) {maxDefault=0.2}, B=Clamp(None=ConstantBiasScale(None=@n45) {bias=0, scale=2}))
  n48 = Transform(None=CameraVectorWS()) {transformSourceType=TRANSFORMSOURCE_World, transformType=TRANSFORM_Tangent}
  n49 = NamedRerouteUsage()
  n50 = Power(Base=OneMinus(None=DotProduct(A=@n48, B=@n49))) {constExponent=6}
  n51 = Clamp(None=@n50) {minDefault=0.005}
  n52 = ScalarParameter() {group=Fresnel, parameterName=DiffuseIntensity_Fresnel}
  n53 = LinearInterpolate(A=Multiply(A=@n1.RGB, B=Constant() {r=1}) {desc=Diffuse}, B=@n8, Alpha=Saturate(None=Multiply(A=@n10, B=@n52))) {constAlpha=1, desc=Diffuse}
  n54 = Constant3Vector() {constant=<rgba 0.03 0.03 0.03 0>}
  n55 = NamedRerouteUsage()
  n56 = Multiply(A=@n55, B=ScalarParameter() {defaultValue=0.2, group=Spear, parameterName=speed})
  n57 = ConstantBiasScale(None=Sine(None=@n56)) {bias=0.65, scale=4}
  n58 = Add(A=TextureCoordinate() {coordinateIndex=1}, B=AppendVector(A=Constant(), B=Clamp(None=@n57) {maxDefault=2, minDefault=0.105}))
  n59 = Cosine(None=@n56)
  n60 = OneMinus(None=Clamp(None=ConstantBiasScale(None=@n59) {bias=-0.5, scale=10}))
  n61 = TextureSampleParameter2D(UVs=@n58) {parameterName=RGBA_Mask, texture=&T_Hug2_RGB}
  n62 = TextureSampleParameter2D(UVs=Panner(Coordinate=TextureCoordinate() {coordinateIndex=1, uTiling=0.1, vTiling=0.15}, Time=@n55) {speedX=0.1}) {parameterName=RGBA_Mask, texture=&T_Hug2_RGB}
  n63 = ConstantBiasScale(None=Multiply(A=Multiply(A=@n60, B=@n61.G), B=ConstantBiasScale(None=@n62.B) {bias=-0.2, scale=1.25})) {bias=0, scale=0.12}
  n64 = Constant()
  n65 = TextureSampleParameter2D(UVs=Add(A=@n58, B=AppendVector(A=@n63, B=@n64))) {parameterName=RGBA_Mask, texture=&T_Hug2_RGB}
  n66 = TextureCoordinate() {coordinateIndex=1}
  n67 = ScalarParameter() {defaultValue=1.1, group=Spear, parameterName=SpearRotate}
  n68 = ComponentMask(None=Rotator(Coordinate=@n66, Time=@n67)) {r=true}
  n69 = ComponentMask(None=Rotator(Coordinate=OneMinus(None=@n66), Time=Multiply(A=@n67) {constB=-1})) {r=true}
  n70 = Clamp(None=ConstantBiasScale(None=ComponentMask(None=@n66) {r=true}) {bias=-0.5, scale=10000})
  n71 = NamedRerouteUsage()
  n72 = Multiply(
   A=Add(A=@n71, B=ScalarParameter() {defaultValue=-0.2, group=Spear, parameterName=timing offset})
   B=ScalarParameter() {defaultValue=0.2, group=Spear, parameterName=speed}
  )
  n73 = ConstantBiasScale(None=Sine(None=@n72)) {bias=0.7, scale=7}
  n74 = Add(
   A=AppendVector(A=LinearInterpolate(A=@n68, B=@n69, Alpha=@n70), B=ComponentMask(None=@n66) {g=true})
   B=AppendVector(A=ScalarParameter() {defaultValue=0.2, group=Spear, parameterName=spear offset}, B=Clamp(None=@n73) {maxDefault=2, minDefault=0.15})
  )
  n75 = Cosine(None=@n72)
  n76 = OneMinus(None=Clamp(None=ConstantBiasScale(None=@n75) {bias=-0.5, scale=10}))
  n77 = TextureSampleParameter2D(UVs=@n74) {parameterName=RGBA_Mask, texture=&T_Hug2_RGB}
  n78 = TextureSampleParameter2D(UVs=Panner(Coordinate=TextureCoordinate() {coordinateIndex=1, uTiling=0.1, vTiling=0.15}, Time=@n71) {speedX=-0.05}) {parameterName=RGBA_Mask, texture=&T_Hug2_RGB}
  n79 = ConstantBiasScale(None=Multiply(A=Multiply(A=@n76, B=@n77.G), B=ConstantBiasScale(None=@n78.B) {bias=-0.2, scale=1.25})) {bias=0, scale=0.12}
  n80 = Constant()
  n81 = TextureSampleParameter2D(UVs=Add(A=@n74, B=AppendVector(A=@n79, B=@n80))) {parameterName=RGBA_Mask, texture=&T_Hug2_RGB}
  n82 = Max(A=@n65.R, B=@n81.R) {desc=Spears - Red Channels}
  n83 = Clamp(None=ConstantBiasScale(None=@n82) {bias=0, scale=30})
  n84 = VectorParameter() {group=Spear, parameterName=Color, defaultValue=<rgba 0.2983 0.7064 1 1>}
  n85 = StaticSwitchParameter(
   True=LinearInterpolate(
    A=LinearInterpolate(A=@n14, B=@n16, Alpha=@n40) {desc=Diffuse}
    B=LinearInterpolate(A=@n16, B=@n15.RGB, Alpha=@n46)
    Alpha=LinearInterpolate(A=@n47, B=@n51) {constAlpha=0.95}
   ) {desc=Diffuse}
   False=LinearInterpolate(A=LinearInterpolate(A=@n53, B=@n54, Alpha=@n83) {constAlpha=1, desc=Diffuse}, B=@n84.RGB, Alpha=@n82) {constAlpha=1, desc=Diffuse}
  ) {defaultValue=true, desc="t=ice, f=spear", group=BaseTextures, parameterName=Ice or Spear}
  n86 = NamedRerouteUsage()
  n87 = Transform(None=CameraVectorWS()) {transformSourceType=TRANSFORMSOURCE_World, transformType=TRANSFORM_Tangent}
  n88 = NamedRerouteUsage()
  n89 = Power(Base=OneMinus(None=DotProduct(A=@n87, B=@n88))) {constExponent=3}
  n90 = OneMinus(None=@n89)
  n91 = ScalarParameter() {defaultValue=1, group=Finish, parameterName=FinishInteriorFalloff}
  n92 = ScalarParameter() {defaultValue=1, group=Finish, parameterName=FinishOuterFalloff}
  n93 = Max(A=Saturate(None=Power(Base=@n90, Exp=@n91)), B=Saturate(None=Power(Base=@n89, Exp=@n92)))
  n94 = ScalarParameter() {defaultValue=95, group=Finish, parameterName=MetallicFinishInteriorFalloff}
  n95 = Clamp(None=Power(Base=OneMinus(None=@n89), Exp=@n94), Min=ScalarParameter() {defaultValue=0.1, group=Finish, parameterName=MetallicDarkMinLevel})
  n96 = ScalarParameter() {defaultValue=5, group=Finish, parameterName=MetallicFinishOuterFalloff}
  n97 = Max(A=@n95, B=Saturate(None=Power(Base=@n89, Exp=@n96)))
  n98 = Power(Base=@n95, Exp=@n94)
  n99 = ScalarParameter() {defaultValue=0.1, group=Finish, parameterName=MetallicHighlight}
  n100 = LinearInterpolate(A=Multiply(A=@n85, B=Multiply(A=@n86, B=@n93)), B=Add(A=Multiply(A=@n85, B=@n97), B=Multiply(A=@n98, B=@n99)), Alpha=NamedRerouteUsage())
  NamedRerouteDeclaration(None=Saturate(None=ConstantBiasScale(None=@n100) {bias=-0.01, desc=Color correction for BR, scale=1.01})) {name=BaseColor, nodeColor=<rgba 1 0 0.375 1>}
  n101 = Clamp(None=Max(A=ConstantBiasScale(None=@n30.A) {bias=-0.01, scale=1}, B=ConstantBiasScale(None=@n37.A) {bias=-0.01, scale=1}))
  n102 = TextureCoordinate() {coordinateIndex=1, uTiling=0.5, vTiling=0.3}
  n103 = Constant2Vector() {g=2, r=3}
  n104 = NamedRerouteUsage()
  n105 = TextureSampleParameter2D(UVs=Panner(Coordinate=Multiply(A=@n102, B=@n103), Time=@n104) {speedX=-0.02, speedY=0.1}) {parameterName=Noise, texture=&T_Hug2_Noise}
  n106 = Multiply(A=@n101, B=@n105.B)
  n107 = Max(A=@n30.B, B=@n37.B) {desc=Icicles - BlueChannels}
  n108 = TextureSampleParameter2D(UVs=Panner(Coordinate=@n102, Time=@n104) {speedX=-0.02, speedY=0.1}) {parameterName=Noise, texture=&T_Hug2_Noise}
  n109 = TextureSampleParameter2D(UVs=Panner(Coordinate=TextureCoordinate() {coordinateIndex=1}, Time=@n104) {speedX=-0.02, speedY=0.1}) {parameterName=Noise, texture=&T_Hug2_Noise}
  n110 = Add(A=Multiply(A=@n101, B=ConstantBiasScale(None=@n108.A) {bias=-0.5, scale=2}), B=Subtract(A=@n101, B=ConstantBiasScale(None=@n109.R) {bias=0}))
  n111 = Max(A=Max(A=@n39, B=ConstantBiasScale(None=@n106) {bias=0, scale=100}), B=ConstantBiasScale(None=Max(A=@n107, B=@n110)) {bias=0, scale=2})
  n112 = LinearInterpolate(
   A=LinearInterpolate(A=@n14, B=@n16, Alpha=Clamp(None=@n111)) {desc=Emissive}
   B=@n15.RGB
   Alpha=Clamp(None=ConstantBiasScale(None=@n106) {bias=0, scale=20}) {maxDefault=2}
  ) {desc=Emissive}
  n113 = Multiply(A=Clamp(None=@n50) {minDefault=0.005}, B=@n47)
  n114 = ConstantBiasScale(None=@n50) {bias=-0.5, scale=2}
  n115 = Clamp(None=@n43)
  n116 = Clamp(None=Max(A=ConstantBiasScale(None=@n113) {bias=0, scale=5}, B=Multiply(A=@n114, B=@n115)))
  n117 = Sine(None=@n42) {period=0.5}
  n118 = ConstantBiasScale(None=OneMinus(None=@n117)) {bias=1.5}
  n119 = TextureSampleParameter2D(UVs=TextureCoordinate()) {parameterName=RGBA_Mask, texture=&T_Hug2_RGB}
  n120 = Multiply(A=OneMinus(None=@n50), B=Multiply(A=NamedRerouteUsage(), B=Multiply(A=@n118, B=@n119.G)))
  n121 = LinearInterpolate(
   A=LinearInterpolate(A=LinearInterpolate(A=@n112, B=@n15.RGB, Alpha=@n116) {desc=Emissive}, B=@n15.RGB, Alpha=@n38) {desc=Emissive}
   B=@n15.RGB
   Alpha=Clamp(None=ConstantBiasScale(None=@n120) {bias=-0.02, scale=5}) {maxDefault=2}
  ) {desc=Emissive}
  n122 = NamedRerouteUsage()
  n123 = ConstantBiasScale(None=@n117) {bias=1.5}
  n124 = Add(A=@n119.G, B=Multiply(A=@n105.B, B=ConstantBiasScale(None=@n119.G) {bias=0, scale=2}))
  n125 = Subtract(A=@n119.G, B=ConstantBiasScale(None=@n109.A) {bias=0, scale=0.2})
  n126 = Clamp(None=Max(A=Multiply(A=@n123, B=@n124), B=ConstantBiasScale(None=@n125) {bias=-0.05, scale=2}))
  n127 = LinearInterpolate(A=@n14, B=ConstantBiasScale(None=@n84.RGB) {bias=0, scale=0.25}, Alpha=@n83) {desc=emissive}
  n128 = NamedRerouteUsage()
  n129 = Floor(None=@n58)
  n130 = Clamp(None=ConstantBiasScale(None=ComponentMask(None=@n129) {g=true}))
  n131 = ConstantBiasScale(None=@n59) {bias=-0.25, scale=2}
  n132 = Multiply(A=Multiply(A=OneMinus(None=@n130), B=@n61.A), B=OneMinus(None=Clamp(None=@n131)))
  n133 = Floor(None=@n74)
  n134 = Clamp(None=ConstantBiasScale(None=ComponentMask(None=@n133) {g=true}))
  n135 = ConstantBiasScale(None=@n75) {bias=0, scale=2}
  n136 = Multiply(A=Multiply(A=OneMinus(None=@n134), B=@n77.A), B=OneMinus(None=Clamp(None=@n135)))
  n137 = Max(A=@n132, B=@n136) {desc=Spears - AlphaFade1}
  n138 = ConstantBiasScale(None=@n137) {bias=0, scale=0.25}
  n139 = NamedRerouteUsage()
  n140 = TextureSampleParameter2D(UVs=Panner(Coordinate=TextureCoordinate() {coordinateIndex=1, vTiling=0.02}) {speedX=0.2}) {parameterName=RGBA_Mask, texture=&T_Hug2_RGB}
  n141 = TextureSampleParameter2D(UVs=Panner(Coordinate=TextureCoordinate() {coordinateIndex=1, vTiling=0.05}) {speedX=-0.2}) {parameterName=RGBA_Mask, texture=&T_Hug2_RGB}
  n142 = Abs(None=ConstantBiasScale(None=Subtract(A=@n140.B, B=@n141.B)) {bias=-0.1})
  n143 = ConstantBiasScale(None=Subtract(A=@n137, B=Clamp(None=@n142))) {bias=-0.12, scale=0.25}
  n144 = TextureSampleParameter2D(UVs=Panner(Coordinate=TextureCoordinate() {coordinateIndex=1, uTiling=6, vTiling=1.5}, Time=@n71) {speedY=1}) {parameterName=Noise, texture=&T_Hug2_Noise}
  n145 = TextureSampleParameter2D(UVs=Panner(Coordinate=TextureCoordinate() {coordinateIndex=1, uTiling=2, vTiling=0.6}, Time=@n71) {speedX=0.2, speedY=1}) {parameterName=Noise, texture=&T_Hug2_Noise}
  n146 = TextureCoordinate() {coordinateIndex=1, uTiling=2, vTiling=0.2}
  n147 = OneMinus(None=TextureSampleParameter2D(UVs=Panner(Coordinate=@n146, Time=@n71) {speedY=0.4}) {parameterName=Noise, texture=&T_Hug2_Noise}.A)
  n148 = Multiply(A=@n137, B=Multiply(A=@n145.G, B=@n147))
  n149 = LinearInterpolate(
   A=LinearInterpolate(A=LinearInterpolate(A=@n127, B=@n128, Alpha=@n138) {desc=emissive}, B=@n139, Alpha=Clamp(None=@n143)) {desc=emissive}
   B=LinearInterpolate(A=@n128, B=@n139, Alpha=ConstantBiasScale(None=@n144.A) {bias=0.5, scale=1})
   Alpha=Clamp(None=ConstantBiasScale(None=@n148) {bias=-0.02, scale=15})
  ) {desc=emissive}
  n150 = Constant3Vector()
  n151 = ConstantBiasScale(None=@n83) {bias=0, scale=0.9}
  n152 = ConstantBiasScale(None=@n82) {bias=0, scale=1.5}
  n153 = Clamp(None=ConstantBiasScale(None=@n59) {bias=-0.25, scale=2})
  n154 = Clamp(None=ConstantBiasScale(None=@n75) {bias=0, scale=2})
  n155 = Max(A=Multiply(A=@n65.R, B=OneMinus(None=@n153)), B=Multiply(A=@n81.R, B=OneMinus(None=@n154)))
  n156 = LinearInterpolate(
   A=LinearInterpolate(A=LinearInterpolate(A=@n149, B=@n150, Alpha=@n151) {desc=emissive}, B=@n84.RGB, Alpha=Clamp(None=@n152) {maxDefault=1000}) {desc=emissive}
   B=NamedRerouteUsage()
   Alpha=Clamp(None=ConstantBiasScale(None=@n155) {bias=-0.3, scale=2})
  ) {desc=emissive}
  n157 = NamedRerouteUsage()
  n158 = Multiply(A=@n137, B=Add(A=Multiply(A=@n147, B=@n144.B), B=Multiply(A=@n144.R, B=@n145.B)))
  n159 = ConstantBiasScale(None=Subtract(A=@n132, B=@n61.G)) {bias=-0.3, scale=2}
  n160 = ConstantBiasScale(None=Subtract(A=@n136, B=ConstantBiasScale(None=@n77.G) {bias=0, scale=60}) {constB=0}) {bias=-0.3, scale=1}
  n161 = Clamp(None=Max(A=ConstantBiasScale(None=@n158) {bias=-0.1, scale=10}, B=Max(A=@n159, B=@n160) {constB=0}) {constB=0}) {maxDefault=5}
  n162 = Saturate(None=StaticSwitchParameter(
    True=LinearInterpolate(A=@n121, B=@n122, Alpha=@n126) {desc=Emissive}
    False=LinearInterpolate(A=@n156, B=@n157, Alpha=@n161) {desc=emissive}
   ) {defaultValue=true, desc="t=ice, f=spear", group=BaseTextures, parameterName=Ice or Spear})
  n163 = ScalarParameter() {defaultValue=3, group=Colors, parameterName=EmissiveIntensity}
  NamedRerouteDeclaration(None=ConstantBiasScale(None=Multiply(A=@n162, B=@n163)) {bias=-0.015, desc=Color correction for BR, scale=1.15}) {name=EmissiveColor, nodeColor=<rgba 0 0.6953 1 1>}
  n164 = TextureSampleParameter2D(UVs=TextureCoordinate()) {
   group=BaseTextures
   parameterName=NormalMap
   samplerSource=SSM_Wrap_WorldGroupSettings
   samplerType=SAMPLERTYPE_Normal
   texture=&/T_Hug_N
  }
  NamedRerouteDeclaration(None=StaticSwitchParameter(True=@n164.RGB, False=@n164.RGB) {defaultValue=true, desc="t=ice, f=spear", group=BaseTextures, parameterName=Ice or Spear}) {name=Normals, nodeColor=<rgba 0.25 0.25 1 1>}
  n165 = TextureCoordinate() {coordinateIndex=1}
  n166 = VectorParameter() {desc="(R,G) = Detail1    (B,A) = Detail2", group=BaseTextures, parameterName=Universal_Map_Transform, defaultValue=<rgba 1 1 0 0>}
  n167 = ComponentMask(None=@n166.RGB) {g=true, r=true}
  NamedRerouteDeclaration(None=Add(A=Multiply(A=@n165, B=@n167), B=AppendVector(A=@n166.B, B=@n166.B)) {desc=Panner}) {name=DiffuseNormal UVs, nodeColor=<rgba 0.1484 0 1 1>}
  n168 = VectorParameter() {desc="Alpha = Emissive Color Brightness", group=Colors, parameterName=BaseColor, defaultValue=<rgba 1 0.2673 0 1>}
  NamedRerouteDeclaration(None=@n168.RGB) {name=TeamColor, nodeColor=<rgba 0 0.5312 1 1>}
  NamedRerouteDeclaration(None=VectorParameter() {group=Colors, parameterName=AccentColor, defaultValue=<rgba 1 0 0 1>}.RGB) {name=CustomColor, nodeColor=<rgba 0 1 0.0625 1>}
  NamedRerouteDeclaration(None=Time() {bOverride_Period=true, period=500}) {name=time, nodeColor=<rgba 1 0 0.4219 1>}
  NamedRerouteDeclaration(None=@n1.A) {name=DiffuseAlpha, nodeColor=<rgba 1 0.7734 0 1>}
  n169 = TextureSampleParameter2D(UVs=TextureCoordinate()) {
   desc="R=specular, G=ao, B=Roughness, A = Metallic"
   group=BaseTextures
   parameterName=AmbientOcclusion
   texture=/VehicleCosmetics/UniversalSkins/Endit/T_endit_AO_Fire
  }
  n170 = StaticSwitchParameter(True=@n169.B, False=ScalarParameter() {defaultValue=0.5, group=Finish, parameterName=Roughness}) {desc=Roughness in B channel of AO map, group=BaseTextures, parameterName=Rough Map}
  n171 = LinearInterpolate(A=@n170, Alpha=@n47) {constB=0}
  NamedRerouteDeclaration(None=Saturate(None=StaticSwitchParameter(True=@n171, False=@n170) {defaultValue=true, desc="t=ice, f=spear", group=BaseTextures, parameterName=Ice or Spear})) {name=Roughness, nodeColor=<rgba 0 0.1094 1 1>}
  NamedRerouteDeclaration(None=StaticSwitchParameter(True=@n169.A, False=ScalarParameter() {group=Finish, parameterName=MetallicDefault}) {desc=Metallic in A channel of AO map, group=BaseTextures, parameterName=Metallic Map}) {name=Metallic, nodeColor=<rgba 0.4297 1 0 1>}
  NamedRerouteDeclaration(None=StaticSwitchParameter(True=@n169.R, False=ScalarParameter() {defaultValue=1, group=Finish, parameterName=SpecularHighlightIntensity}) {desc="Spec in R channel of AO map (helps Finish consistency)", group=BaseTextures, parameterName=SpecMap}) {name=Specular, nodeColor=<rgba 1 0.2578 0 1>}
  n172 = Constant3Vector() {constant=<rgba 1 1 1 0>}
  n173 = ScalarParameter() {defaultValue=1, group=BaseTextures, parameterName=AO}
  NamedRerouteDeclaration(None=StaticSwitchParameter(True=LinearInterpolate(A=@n172, B=@n169.G, Alpha=@n173), False=@n172) {desc=AO in G channel of AO map, group=BaseTextures, parameterName=AmbientOcclusionMap}) {name=AO, nodeColor=<rgba 0.7969 0.7969 0.7969 1>}
  SetMaterialAttributes!(
   MaterialAttributes
   Base Color=NamedRerouteUsage()
   Metallic=NamedRerouteUsage()
   Roughness=NamedRerouteUsage()
   Emissive Color=NamedRerouteUsage()
   Normal=NamedRerouteUsage()
   Ambient Occlusion=NamedRerouteUsage()
  ) {attributeSetTypes [
    #1
    #2
    #3
    #4
    #5
    #6
   ]}
 ]
}
